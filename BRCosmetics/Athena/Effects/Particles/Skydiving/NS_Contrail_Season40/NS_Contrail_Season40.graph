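# Niagara skydiving contrail cosmetic for Season 40, with red and gold trails, spinning mesh splines, flares with lights, and stretched sprites on the character.
# NE_LocalSpaceRibbon builds pendulum-constrained ribbon chains from skeletal mesh sockets, swept against owner velocity, with a front-end variant.
# Spline, flare and tertiary emitters sample those chains through Particle Attribute Readers, some traveling along them.
# Dive values scale spawn rates, sizes, spin and brightness.
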NiagaraSystem /BRCosmetics/Athena/Effects/Particles/Skydiving/NS_Contrail_Season40/NS_Contrail_Season40
dependencies {
 usedDataInterfaces [
  &NiagaraDataInterfaceCurve
  &NiagaraDataInterfaceColorCurve
  &NiagaraDataInterfaceParticleRead
  &NiagaraDataInterfaceVectorCurve
  &NiagaraDataInterfaceArrayFloat
  &NiagaraDataInterfaceCurlNoise
  &NiagaraDataInterfaceArrayColor
  &NiagaraDataInterfaceSpriteRendererInfo
  &NiagaraDataInterfaceVector2DCurve
  &NiagaraDataInterfaceSkeletalMesh
 ]
 usedDynamicInputs [/Niagara/DynamicInputs/Lerp/SmoothLerpOverTime_Float]
 usedModules [
  &self:SystemSpawnScript.SystemScriptSource.SystemScriptGraph.NiagaraNodeAssignment_1.SetVariables_#1
  &self:SystemSpawnScript.SystemScriptSource.SystemScriptGraph.NiagaraNodeAssignment_0.SetVariables_#2
  &self:SystemSpawnScript.SystemScriptSource.SystemScriptGraph.NiagaraNodeAssignment_4.SetVariables_#3
  &self:SystemSpawnScript.SystemScriptSource.SystemScriptGraph.NiagaraNodeAssignment_3.SetVariables_#4
  &self:SystemSpawnScript.SystemScriptSource.SystemScriptGraph.NiagaraNodeAssignment_2.SetVariables_#5
  &SystemState
  &InterpolateOverTime
 ]
}
emitters {
 DigitalSquares001 {
  emitterUpdateScript [
   &EmitterState(
    Inactive Response=::NewEnumerator1
    Life Cycle Mode=::NewEnumerator1
    Loop Behavior=::NewEnumerator1
    Loop Duration Mode=::NewEnumerator1
    UseLoopDelay=false
   )
   &SpawnRate(SpawnRate=&Lerp_Float(A=60, B=130, Alpha=$System.FN_Contrail_Setup.Dive Value), Use Spawn Probability=false) {module=SpawnRate002}
  ]
  particleSpawnScript [
   &PartitionParticles()
   &InitializeParticle(
    Color=&MultiplyLinearColor(
     Linear Color=&RandomRangeLinearColor(Minimum=$User.FXColorB, Maximum=$User.FXColorA)
     Scale Factor=<5 5 5 1>
    )
    Lifetime=&Lerp_Float(
     A=&RandomRangeFloat(Minimum=0.1, Maximum=0.4)
     B=&RandomRangeFloat(Minimum=0.1, Maximum=0.15)
     Alpha=$System.FN_Contrail_Setup.Dive Value
    )
    Sprite Size Mode=::NewEnumerator1
    Uniform Sprite Size Max=4
    Uniform Sprite Size Min=0.5
    UsePositionOffset=false
    Write Material Random=true
   )
   &SkeletalMeshLocation(
    Bone / Socket Index=$Particles.PartitionParticles.Partition
    Bone Sampling Mode=::NewEnumerator4
    MaskSpawnGroup=false
    Orientation Sampling=::NewEnumerator1
    Skeletal Mesh=$System.Skeletal Mesh
   )
   &ShapeLocation(MaskSpawnGroup=false, Offset=<0 0 -15>, Offset Mode=::NewEnumerator0, Override Seed=false, Sphere Radius=15, UseOwnerScale=true)
   &AddVelocity(
    Velocity=&Lerp_Vector(A=<0 0.1 1>, B=<0 3 1>, Alpha=$System.FN_Contrail_Setup.Dive Value)
    Velocity Speed Scale=50
   )
  ]
  particleUpdateScript [
   &ParticleState()
   &UpdateMeshOrientation(
    Facing Direction=<0 0 1>
    Facing Interpolation=false
    Orientation Method=::NewEnumerator1
    Override Quat=false
    Stabilization Mode:ENiagara_OrientVectorCount=::NewEnumerator2
    TransformFacing=false
    TransformSide=false
   )
   &AccelerationForce(Acceleration=&Lerp_Vector(
     A=&RandomRangeVector(Minimum=<100 100 900>, Maximum=<-100 -100 1500>)
     B=&RandomRangeVector(Minimum=<0 0 3000>, Maximum=<0 0 5555>)
     Alpha=$System.FN_Contrail_Setup.Dive Value
    )) {module=AccelerationForce001}
   &Drag(
    Drag=&FloatFromCurve(
     FloatCurve=DataInterface() {propertyValues {Curve {
        defaultValue=1
        keys [
         {arriveTangent=2.2139, interpMode=RCIM_Cubic, leaveTangent=2.2139, tangentMode=RCTM_User, time=0, value=0}
         {interpMode=RCIM_Cubic, tangentMode=RCTM_User, time=1, value=1}
        ]
       }}}
     Scale Curve=&RandomRangeFloat(Minimum=22, Maximum=22)
    )
    Use Linear Drag=true
    Use Rotational Drag=true
   )
   &SolveForcesAndVelocity(Clamp Velocity=false, Limit Acceleration=false)
   &AlignSpriteToMeshOrientation(
    Mesh Orientation Relative Sprite Alignment Vector=<1 0 0>
    Mesh Orientation Relative Sprite Facing Vector:Vector3f=<0 1 0>
    SetAlignment=true
    SetFacing=false
   )
   &ScaleSpriteSizeBySpeed(
    Max Scale Factor=&Multiply_Vector2D_ByFloat(
     Float=1
     Vector2D=&Multiply_Vector2D(
      A=&Vector2DFromFloat(Value=0.5)
      B=&Lerp_Vector2(A=<0.01 1>, B=<0.01 200>, Alpha=$System.FN_Contrail_Setup.Dive Value)
     )
    )
    Min Scale Factor=<0.5 1.5>
    Velocity Threshold=2000
   )
   &ScaleSpriteSize(
    Non-Uniform Curve Scale=&Multiply_Vector2D_ByFloat(
     Float=&FloatFromCurve(FloatCurve=DataInterface() {propertyValues {Curve {
         defaultValue=0
         keys [
          <at 0 0 RCIM_Cubic>
          {arriveTangent=2, time=0.1, value=1}
          {interpMode=RCIM_Cubic, tangentMode=RCTM_Break, time=0.5, value=1}
          {interpMode=RCIM_Cubic, tangentMode=RCTM_User, time=1, value=0}
         ]
        }}})
     Vector2D=<1 1>
    )
    Non-Uniform Curve Sprite Scale=DataInterface() {propertyValues {
      XCurve {keys [{interpMode=RCIM_Cubic, tangentMode=RCTM_User, time=0, value=1}, {tangentMode=RCTM_User, time=1, value=0}]}
      YCurve {keys [{interpMode=RCIM_Cubic, tangentMode=RCTM_User, time=0.5, value=1}, {tangentMode=RCTM_User, time=1, value=0}]}
     }}
    Scale Sprite Size Mode=::NewEnumerator3
   )
   &CameraOffset(Camera Offset Amount=&RandomRangeFloat(Minimum=-15, Maximum=-3))
   /SetVariables(
    Particles.GlowColor=&MultiplyLinearColor(
     Linear Color=&MultiplyLinearColor(
      Linear Color=&RandomRangeLinearColor(Minimum=$User.FXColorB, Maximum=$User.FXColorA)
      Scale Factor=<2 2 2 1>
     )
     Scale Factor=&MakeVector4(
      X=1
      Y=1
      Z=1
      W=&Multiply_Float(
       A=&RandomRangeFloat(Minimum=0.08, Maximum=0.05)
       B=&FloatFromCurve(FloatCurve=DataInterface() {propertyValues {Curve {
           defaultValue=1
           keys [{interpMode=RCIM_Cubic, tangentMode=RCTM_User, time=0.5, value=1}, {interpMode=RCIM_Cubic, tangentMode=RCTM_User, time=1, value=0}]
          }}})
      )
     )
    )
    Particles.GlowSize=&Multiply_Vector2D_ByFloat(Float=22, Vector2D=<1 1>)
   ) {isSetParameters=true}
   &Color(
    Color=&Multiply_LinearColorByFloat(
     Linear Color=&LerpLinearColors(
      StartColor=&SelectLinearColorFromArray(
       Array Sampling Mode=::NewEnumerator0
       Direct Array Index=?
       Color Selection Array=DataInterface() {propertyValues {ColorData [<rgba 3 0 0 1>, <rgba 1 0.7441 0 1>]}}
       Random Seed=0
      )
      EndColor=$User.FXColorA
      LerpFactor=0
     )
     Float=&FloatFromCurve(
      FloatCurve=DataInterface() {propertyValues {Curve {keys [{interpMode=RCIM_Cubic, tangentMode=RCTM_User, time=0, value=2}, <at 1 10>]}}}
      Scale Curve=&RandomRangeFloat(Minimum=0.5, Maximum=1)
     )
    )
    Scale Color=&Multiply_VectorByFloat(
     Vector=<1 1 1>
     Float=&FloatFromCurve(FloatCurve=DataInterface() {propertyValues {Curve {keys [{arriveTangent=-17.669, interpMode=RCIM_Cubic, leaveTangent=-17.669, tangentMode=RCTM_User, time=0, value=500}, <at 0.2 1>, <at 1 1>]}}})
    )
   )
   &CameraOffset(Camera Offset Amount=&RandomRangeFloat(Minimum=15, Maximum=-3)) {module=CameraOffset001}
  ]
  rendererClasses [&NiagaraSpriteRendererProperties]
  renderers [{rendererClass=&NiagaraSpriteRendererProperties}]
 }
 ExtraSplines {
  emitterUpdateScript [
   &EmitterState(
    Inactive Response=::NewEnumerator1
    Life Cycle Mode=::NewEnumerator1
    Loop Behavior=::NewEnumerator1
    Loop Duration Mode=::NewEnumerator1
    UseLoopDelay=false
   ) {module=EmitterLifeCycle}
   &SpawnBurst_Instantaneous(Spawn Count=6, Use Spawn Probability=false, UseLoopCountLimit=false)
  ]
  particleSpawnScript [
   &InitializeParticle(Color Mode=::NewEnumerator0, Lifetime=5, UsePositionOffset=false, Write Material Random=true)
   &PartitionParticles(Normalized Index In Partition Includes One=true, Particles Per Partition=3, Partition Mode:ENiagara_PartitionMode=::NewEnumerator1)
   &InitialMeshOrientation(
    Mesh Orientation Mode=::NewEnumerator2
    Orientation Coordinate Space=::Local
    Orientation Vector=<0 0 1>
    Rotation=&Lerp_Vector(
     A=<0 0 0>
     B=&Lerp_Vector(A=<0 0 1>, B=<0 0 -1>, Alpha=&MakeFloatFromInt(Int=$Particles.PartitionParticles.Partition))
     Alpha=$Particles.PartitionParticles.Normalized Index In Partition
    )
    Use Rotation Vector=true
   )
   /SetVariables(Particles.Scale=&Multiply_Vector(
     A=&Lerp_Vector(A=<0.175 0.175 0.45>, B=<0.1 0.1 0.1>, Alpha=$Particles.PartitionParticles.Normalized Index In Partition)
     B=&Lerp_Vector(A=<1 1 1>, B=<1 -1 1>, Alpha=&MakeFloatFromInt(Int=$Particles.PartitionParticles.Partition))
    )) {isSetParameters=true, module=SetVariables_#6}
  ]
  particleUpdateScript [
   &ParticleState(Kill Particles When Lifetime Has Elapsed=false)
   &ScaleMeshSize(Scale Factor=&Lerp_Vector(A=<1 1 1>, B=<1.2 1.2 2>, Alpha=$System.SmoothDiveAlpha))
   &self:SampleLocalRibbon(
    Particle Attribute Reader=DataInterface() {propertyValues {EmitterBinding {bindingMode=Other, emitterName=NE_LocalSpaceRibbon}}}
    ParticlePerRibbon:NiagaraInt32=$System.ParticlePerRibbon
    Partition:NiagaraInt32=$Particles.PartitionParticles.Partition
   ) {module=SampleLocalRibbon}
   &UpdateMeshOrientation(
    Rotation Rate=&Multiply_Float(
     A=&Lerp_Float(A=-1, B=1, Alpha=&MakeFloatFromInt(Int=$Particles.PartitionParticles.Partition))
     B=&Lerp_Float(A=1.6, B=2.5, Alpha=$System.SmoothDiveAlpha)
    )
    Rotation Vector=<0 0 1>
   ) {module=UpdateMeshOrientation001}
   &UpdateMeshOrientation(
    Facing Direction=$Particles.RibbonDirectionVector
    Facing Interpolation=false
    Orientation Method=::NewEnumerator1
    Override Quat=false
    TransformFacing=false
    TransformUp=false
    Up Direction:Vector3f=<0 1 0>
   )
   &SolveForcesAndVelocity(Clamp Velocity=false, Limit Acceleration=false)
   &DynamicMaterialParameters(
    Index 0 Param 1=25
    Index 0 Param 2=2
    Index 0 Param 3=&Multiply_Float(A=$Engine.System.Age, B=1)
    Module.Index0Param0Write=true
    Module.Index0Param1Write=true
    Module.Index0Param2Write=true
    Module.Index0Param3Write=true
   )
   &Color(
    Color=&LerpLinearColors(
     StartColor=$Output.SampleLocalRibbon.RibbonColor
     EndColor=&Multiply_LinearColorByFloat(Linear Color=$Output.SampleLocalRibbon.RibbonColor, Float=1.9684)
     LerpFactor=&Lerp_Float(A=0.1, B=1, Alpha=$System.SmoothDiveAlpha)
    )
    Scale Alpha=&Lerp_Float(A=0.6, B=0.8, Alpha=$System.SmoothDiveAlpha)
   ) {module=Color001}
  ]
  rendererClasses [&NiagaraMeshRendererProperties]
  renderers [{rendererClass=&NiagaraMeshRendererProperties, rendererIndex=0}]
 }
 FeetSpline {
  emitterUpdateScript [
   &EmitterState(
    Inactive Response=::NewEnumerator1
    Life Cycle Mode=::NewEnumerator1
    Loop Behavior=::NewEnumerator1
    Loop Duration Mode=::NewEnumerator1
    UseLoopDelay=false
   ) {module=EmitterLifeCycle}
   &SpawnBurst_Instantaneous(Spawn Count=6, Use Spawn Probability=false, UseLoopCountLimit=false)
  ]
  particleSpawnScript [
   &InitializeParticle(Lifetime=5, UsePositionOffset=false, Write Material Random=true)
   &PartitionParticles(Normalized Index In Partition Includes One=true, Particles Per Partition=3, Partition Mode:ENiagara_PartitionMode=::NewEnumerator1)
   &InitialMeshOrientation(
    Mesh Orientation Mode=::NewEnumerator2
    Orientation Coordinate Space=::Local
    Orientation Vector=<0 0 1>
    Rotation=&Lerp_Vector(A=<0 0 0>, B=<0 0 1>, Alpha=$Particles.PartitionParticles.Normalized Index In Partition)
    Use Rotation Vector=true
   )
   /SetVariables(Particles.Scale=&Multiply_Vector(
     A=&Lerp_Vector(A=<0.1 0.1 0.4>, B=<0.1 0.1 0.2>, Alpha=$Particles.PartitionParticles.Normalized Index In Partition)
     B=&Lerp_Vector(A=<1 -1 1>, B=<1 1 1>, Alpha=&MakeFloatFromInt(Int=$Particles.PartitionParticles.Partition))
    )) {isSetParameters=true, module=SetVariables_#6}
  ]
  particleUpdateScript [
   &ParticleState(Kill Particles When Lifetime Has Elapsed=false)
   &ScaleMeshSize(Scale Factor=&Lerp_Vector(A=<1 1 0.7>, B=<1.5 1.5 1.5>, Alpha=$System.SmoothDiveAlpha))
   &self:SampleLocalRibbon(
    Particle Attribute Reader=DataInterface() {propertyValues {EmitterBinding {bindingMode=Other, emitterName=NE_LocalSpaceRibbon}}}
    ParticlePerRibbon:NiagaraInt32=$System.ParticlePerRibbon
    Partition:NiagaraInt32=$Particles.PartitionParticles.Partition
   ) {module=SampleLocalRibbon}
   &SkeletalMeshLocation(
    Bone / Socket Index=&Add_Integer(A=$Particles.PartitionParticles.Partition, B=2)
    Bone Sampling Mode=::NewEnumerator4
    MaskSpawnGroup=false
    Orientation Sampling=::NewEnumerator1
    Sampled Position Offset=&Lerp_Vector(A=<5 0 2>, B=<-5 0 -2>, Alpha=&MakeFloatFromInt(Int=$Particles.PartitionParticles.Partition))
    Skeletal Mesh=$System.Skeletal Mesh
   ) {module=SkeletalMeshLocation001}
   &UpdateMeshOrientation(
    Rotation Rate=&Multiply_Float(
     A=&Lerp_Float(A=1, B=-1, Alpha=&MakeFloatFromInt(Int=$Particles.PartitionParticles.Partition))
     B=&Lerp_Float(A=1.2, B=2, Alpha=$System.SmoothDiveAlpha)
    )
    Rotation Vector=<0 0 1>
   ) {module=UpdateMeshOrientation001}
   &UpdateMeshOrientation(
    Facing Direction=$Particles.RibbonDirectionVector
    Facing Interpolation=false
    Orientation Method=::NewEnumerator1
    Override Quat=false
    TransformFacing=false
    TransformUp=false
    Up Direction:Vector3f=<0 1 0>
   )
   &SolveForcesAndVelocity(Clamp Velocity=false, Limit Acceleration=false)
   &DynamicMaterialParameters(
    Index 0 Param 1=20
    Index 0 Param 2=1
    Index 0 Param 3=&Multiply_Float(A=$Engine.System.Age, B=1)
    Module.Index0Param0Write=true
    Module.Index0Param1Write=true
    Module.Index0Param2Write=true
    Module.Index0Param3Write=true
   )
   &Color(
    Color=&LerpLinearColors(
     StartColor=$Output.SampleLocalRibbon.RibbonColor
     EndColor=&Multiply_LinearColorByFloat(Linear Color=$Output.SampleLocalRibbon.RibbonColor, Float=1.9684)
     LerpFactor=&Lerp_Float(A=0.1, B=1, Alpha=$System.SmoothDiveAlpha)
    )
    Scale Alpha=&Lerp_Float(A=0.6, B=0.5, Alpha=$System.SmoothDiveAlpha)
   ) {module=Color001}
  ]
  rendererClasses [&NiagaraMeshRendererProperties]
  renderers [{rendererClass=&NiagaraMeshRendererProperties, rendererIndex=0}]
 }
 Hand_Feet_Flares {
  emitterUpdateScript [
   &EmitterState(
    Inactive Response=::NewEnumerator1
    Life Cycle Mode=::NewEnumerator1
    Loop Behavior=::NewEnumerator1
    Loop Duration Mode=::NewEnumerator1
    UseLoopDelay=false
   )
   /Niagara/Noise/Waveform(
    EnableExponent1=false
    GlobalClampMax=false
    GlobalClampMin=false
    UseInput1=false
    UsePhase1=false
    "Waveform [1]":ENiagara_Waveforms=::NewEnumerator12
    "[1] Frequency"=7
   )
   /SetVariables(Emitter.Flicker:NiagaraFloat=&Add_Float(A=&Multiply_Float(A=$Emitter.Waveform.Waveform, B=0.5), B=0.5)) {isSetParameters=true}
   &SpawnBurst_Instantaneous(Spawn Count=2, Use Spawn Probability=false, UseLoopCountLimit=false)
   &SpawnBurst_Instantaneous(Spawn Count=4, Spawn Group=1, Use Spawn Probability=false, UseLoopCountLimit=false) {module=SpawnBurst_Instantaneous001}
  ]
  particleSpawnScript [
   /SetVariables(Particles.VisibilityTag=&SpawnGroup()) {isSetParameters=true}
   &PartitionParticles(Number Of Partitions=&MergeHelper_IntSwitch(0=2, 1=4, 2=?, 3=?, 4=?, 5=?, Index=&SpawnGroup(), Num=::NewEnumerator0))
   &InitializeParticle(
    Color=&MergeHelper_ColorSwitch(0=<rgba 3 0 0 1>, 1=<rgba 2.199 1.3985 0 1>, 2=?, 3=?, Index=$Particles.PartitionParticles.Partition, Num=::NewEnumerator0)
    Lifetime=5
    Sprite Size Mode=::NewEnumerator3
    Uniform Sprite Size=&MergeHelper_FloatSwitch(0=10, 1=50, 2=?, 3=?, Index=&SpawnGroup(), Num=::NewEnumerator0)
    UsePositionOffset=false
    Write Material Random=true
   )
  ]
  particleUpdateScript [
   &ParticleState(Kill Particles When Lifetime Has Elapsed=false)
   &self:SampleLocalRibbon(
    Particle Attribute Reader=DataInterface() {propertyValues {EmitterBinding {bindingMode=Other, emitterName=NE_LocalSpaceRibbon}}}
    ParticlePerRibbon:NiagaraInt32=$System.ParticlePerRibbon
    Partition:NiagaraInt32=$Particles.Initial.PartitionParticles.Partition
   ) {module=SampleLocalRibbon}
   &SkeletalMeshLocation(
    Bone / Socket Index=$Particles.PartitionParticles.Partition
    Bone Sampling Mode=::NewEnumerator4
    MaskSpawnGroup=false
    Sampled Position Offset=&Lerp_Vector(A=<0 0 0>, B=<0 0 0>, Alpha=&MakeFloatFromInt(Int=$Particles.PartitionParticles.Partition))
    Skeletal Mesh=$System.Skeletal Mesh
   ) {module=SkeletalMeshLocation001}
   &ScaleColor(
    Scale RGB=&MakeVectorFromLinearColorRGB(Color=&MultiplyLinearColor(
      Linear Color=<rgba 1 1 1 1>
      Scale Factor=&MakeVector4_FromLinearColor(LinearColor=&LerpLinearColors(
        StartColor=&MergeHelper_ColorSwitch(0=<rgba 3 0 0 1>, 1=<rgba 2.199 1.3985 0 1>, 2=?, 3=?, Index=$Particles.PartitionParticles.Partition, Num=::NewEnumerator0)
        EndColor=&MergeHelper_ColorSwitch(0=<rgba 3 0 0 1>, 1=<rgba 2.199 1.3985 0 1>, 2=?, 3=?, Index=$Particles.PartitionParticles.Partition, Num=::NewEnumerator0)
        LerpFactor=$System.SmoothDiveAlpha
       ))
     ))
    ScaleA=true
    ScaleRGB=true
   )
   &DynamicMaterialParameters(
    Index 0 Param 2=1
    Index 0 Param 4=0.3
    Module.Index0Param0Write=true
    Module.Index0Param1Write=true
    Module.Index0Param2Write=true
    Module.Index0Param3Write=true
   )
   &FN_Light_Attributes(
    Color=$Particles.Color
    Create Color Attribute=true
    Create Exponent Attribute=true
    Create Radius Attribute=true
    Exponent=5
    Radius=100
    Scale Color=&Multiply_Float(A=7, B=$Emitter.Flicker)
   )
   /SetVariables(Particles.Color2=&Multiply_LinearColorByFloat(
     Linear Color=&MergeHelper_ColorSwitch(
      0=&LerpLinearColors(StartColor=<rgba 1.5 0.8527 0 0.2957>, EndColor=<rgba 3 1.6891 0 0.3478>, LerpFactor=$System.SmoothDiveAlpha)
      1=&LerpLinearColors(StartColor=<rgba 3 0.0058 0 0.0522>, EndColor=<rgba 3 0.0058 0 0.0609>, LerpFactor=$System.SmoothDiveAlpha)
      2=&LerpLinearColors(StartColor=<rgba 3 0.0058 0 0.0522>, EndColor=<rgba 3 0.0058 0 0.0609>, LerpFactor=$System.SmoothDiveAlpha)
      3=&LerpLinearColors(StartColor=<rgba 1 0.5685 0 0.2696>, EndColor=<rgba 3 1.6891 0 0.3478>, LerpFactor=$System.SmoothDiveAlpha)
      Index=$Particles.PartitionParticles.Partition
      Num=::NewEnumerator2
     )
     Float=$Emitter.Flicker
    )) {isSetParameters=true}
  ]
  rendererClasses [
   &NiagaraMeshRendererProperties
   &NiagaraLightRendererProperties
   &NiagaraSpriteRendererProperties
  ]
  renderers [
   {rendererClass=&NiagaraMeshRendererProperties, rendererIndex=0}
   {rendererClass=&NiagaraLightRendererProperties, rendererIndex=1}
   {rendererClass=&NiagaraSpriteRendererProperties, rendererIndex=2}
   {rendererClass=&NiagaraSpriteRendererProperties, rendererIndex=3}
  ]
 }
 NE_LocalSpaceRibbon {
  emitterUpdateScript [
   &EmitterState(
    Inactive Response=::NewEnumerator1
    Life Cycle Mode=::NewEnumerator1
    Loop Behavior=::NewEnumerator1
    Loop Duration Mode=::NewEnumerator1
    UseLoopDelay=false
   ) {module=EmitterLifeCycle}
   &SpawnBurst_Instantaneous(
    Spawn Count=&Multiply_Int(A=$System.ChainNumber, B=$System.ParticlePerRibbon)
    Use Spawn Probability=false
    UseLoopCountLimit=false
   )
  ]
  particleSpawnScript [
   &InitializeParticle(Lifetime=5, Sprite Size Max=<20 20>, Sprite Size Mode=::NewEnumerator2, UsePositionOffset=false, Write Material Random=true)
   &PartitionParticles(
    Number Of Partitions=$System.ChainNumber
    Particles Per Partition:NiagaraInt32=$System.ParticlePerRibbon
    Partition Mode:ENiagara_PartitionMode_v2=::NewEnumerator1
   )
   /SetVariables(Particles.NormalizedChainPosition:NiagaraFloat=$Particles.PartitionParticles.Normalized Index In Partition) {isSetParameters=true}
  ]
  particleUpdateScript [
   &ParticleState(Kill Particles When Lifetime Has Elapsed=false)
   &SkeletalMeshLocation(
    Bone / Socket Index=$Particles.PartitionParticles.Partition
    Bone Sampling Mode=::NewEnumerator4
    MaskSpawnGroup=false
    Position Sampling=::NewEnumerator2
    Sampled Position Offset=&Lerp_Vector(A=<2 5 0>, B=<-2 -5 0>, Alpha=&MakeFloatFromInt(Int=$Particles.PartitionParticles.Partition))
    Skeletal Mesh=$System.Skeletal Mesh
   ) {module=SkeletalMeshLocation001}
   &PendulumSetup(
    Module.Enable Spring Driver=true
    Pendulum Length:NiagaraFloat=&Lerp_Float(
     A=0
     B=$System.ChainLength
     Alpha=&FloatFromCurve(
      FloatCurve=DataInterface() {propertyValues {Curve {keys [<at 0 0>, {arriveTangent=-0.0703, interpMode=RCIM_Cubic, leaveTangent=-0.0703, tangentMode=RCTM_User, time=1, value=1}]}}}
      CurveIndex=$Particles.NormalizedChainPosition
      Scale Curve=&FloatFromCurve(
       FloatCurve=DataInterface() {propertyValues {Curve {keys [
           {arriveTangent=0.6092, interpMode=RCIM_Cubic, leaveTangent=0.6092, tangentMode=RCTM_User, time=0, value=0}
           {tangentMode=RCTM_User, time=1, value=1}
          ]}}}
       CurveIndex=$Emitter.Age
      )
     )
    )
    Pendulum Pivot:Vector3f=&AddVectorToPosition(
     Position=&AddVectorToPosition(Position=$Particles.SkeletalMeshLocation001.SampledPosition, Vector=<0 0 0>)
     Vector=&TransformVector(Vector=<0 0 0>, Source Space=::World, Destination Space=::Simulation)
    )
    Pendulum Pivot Offset Coordinate Space:ENiagaraCoordinateSpace=::Local
    Pendulum Rest Axis:Vector3f=&TransformVector(
     Vector=&Add_Vector(
      A=&Lerp_Vector(
       A=&Add_Vector(
        A=&Multiply_VectorByFloat(
         Vector=&NormalizeVector(Vector To Normalize=$Engine.Owner.Velocity, Safe Normalize=true, Safe Fallback Threshold=0, Safe Fallback Vector=<0 0 0.001>)
         Float=-0.5
        )
        B=&TransformVector(
         Vector=&RotateVectorByQuaternion(
          Vector To Rotate=&Multiply_VectorByFloat(
           Vector=<0.5 1 0>
           Float=&Lerp_Float(A=1, B=-1, Alpha=&MakeFloatFromInt(Int=$Particles.PartitionParticles.Partition))
          )
          Quaternion=$Particles.SkeletalMeshLocation001.SampledRotation
         )
         Destination Space=::World
        )
       )
       B=&Lerp_Vector(
        A=&Multiply_VectorByFloat(
         Vector=&NormalizeVector(Vector To Normalize=$Engine.Owner.Velocity, Safe Normalize=true, Safe Fallback Threshold=0, Safe Fallback Vector=<0 0 0.001>)
         Float=-0.5
        )
        B=<0 0 0.1>
        Alpha=$System.SmoothDiveAlpha
       )
       Alpha=&Lerp_Float(
        A=&FloatFromCurve(
         FloatCurve=DataInterface() {propertyValues {Curve {keys [
             {arriveTangent=1.8207, interpMode=RCIM_Cubic, leaveTangent=1.8207, tangentMode=RCTM_User, time=0, value=0}
             {arriveTangent=-0.078, interpMode=RCIM_Cubic, leaveTangent=-0.078, tangentMode=RCTM_User, time=1, value=1}
            ]}}}
         CurveIndex=$Particles.NormalizedChainPosition
        )
        B=1
        Alpha=&FN_FrontEndFloatModifier(Gameplay=0, Front End=1)
       )
      )
      B=&TransformVector(
       Vector=&FN_FrontEndVector3Modifier(Gameplay=<0 0 0>, Front End=&Lerp_Vector(A=<0 -0.1 0.2>, B=<0 -0.2 0.2>, Alpha=$System.SmoothDiveAlpha))
       Destination Space=::World
      )
     )
     Source Space=::World
     Destination Space=::Simulation
    )
    "Spring Driven Constraint (Tightness)"=30
   )
   &LinearForce(
    Coordinate Space:ENiagaraCoordinateSpace=::World
    Force=&Multiply_VectorByFloat(Vector=$Engine.Owner.Velocity, Float=-5)
   )
   &SolveForcesAndVelocity(Clamp Velocity=false, Limit Acceleration=false)
   &PendulumConstraint()
   &CalculateAccurateVelocity()
   /SetVariables(Particles.RibbonWidth=&FloatFromCurve(
     FloatCurve=DataInterface() {propertyValues {Curve {keys [
         {arriveTangent=8.0134, interpMode=RCIM_Cubic, leaveTangent=8.0134, tangentMode=RCTM_User, time=0, value=0.7}
         {arriveTangent=0.3471, interpMode=RCIM_Cubic, leaveTangent=0.3471, tangentMode=RCTM_User, time=0.1, value=1.2}
         {arriveTangent=-0.7173, interpMode=RCIM_Cubic, leaveTangent=-0.7173, tangentMode=RCTM_User, time=1, value=2}
        ]}}}
     CurveIndex=$Particles.NormalizedChainPosition
     Scale Curve=&Lerp_Float(A=40, B=60, Alpha=$System.SmoothDiveAlpha)
    )) {isSetParameters=true}
   &DynamicMaterialParameters(
    Index 0 Param 3=1
    Index 0 Param 4=&Add_Float(
     A=&Multiply_Float(A=$Engine.System.Age, B=1.2)
     B=&Multiply_FloatByInt(Float=1.6902, Integer=$Particles.PartitionParticles.Partition)
    )
    Index 1 Param 1=1
    Index 1 Param 2=&Add_Float(
     A=&Multiply_Float(A=$Engine.System.Age, B=1.5786)
     B=&Multiply_FloatByInt(Float=1.6902, Integer=$Particles.PartitionParticles.Partition)
    )
    Index 1 Param 3=1
    Index 3 Param 1=1
    Index 3 Param 2=1
    Index 3 Param 3=1
    Module.Index0Param0Write=true
    Module.Index0Param1Write=true
    Module.Index0Param2Write=true
    Module.Index0Param3Write=true
    Module.Index1Param0Write=true
    Module.Index1Param1Write=true
    Module.Index1Param2Write=true
    Module.Index1Param3Write=true
    Module.Index3Param0Write=true
    Module.Index3Param1Write=true
    Module.Index3Param2Write=true
    Module.Index3Param3Write=true
    Write Parameter Index 1=true
    Write Parameter Index 3=true
   )
   &self:Ribbon3Location(
    Amplitude:Vector3f=&Multiply_VectorByFloat(
     Vector=<18 18 1>
     Float=&FloatFromCurve(
      FloatCurve=DataInterface() {propertyValues {Curve {keys [
          {interpMode=RCIM_Cubic, tangentMode=RCTM_User, time=0, value=0}
          {arriveTangent=-2.3116, interpMode=RCIM_Cubic, leaveTangent=-2.3116, tangentMode=RCTM_User, time=0.4, value=1}
          <at 1 0>
         ]}}}
      CurveIndex=$Particles.PartitionParticles.Normalized Index In Partition
      Scale Curve=&Lerp_Float(A=1, B=1, Alpha=$System.SmoothDiveAlpha)
     )
    )
    Partition:NiagaraFloat=&MakeFloatFromInt(Int=$Particles.PartitionParticles.Partition)
    Scale:Vector3f=&Multiply_VectorByFloat(Vector=<1 1 1>, Float=0.5)
    Time:NiagaraFloat=&Add_Float(
     A=&Multiply_Float(A=$Engine.System.Age, B=-1.5)
     B=$Particles.PartitionParticles.Normalized Index In Partition
    )
   ) {module=Ribbon3Location}
   /SetVariables(
    Particles.Color02:LinearColor=&MergeHelper_ColorSwitch(
     0=&ColorFromCurve(
      ColorCurve=DataInterface() {propertyValues {
        AlphaCurve {keys [<at 0 1 RCIM_Cubic>]}
        BlueCurve {keys [<at 0.0572 0 RCIM_Cubic>,<at 0.4612 0 RCIM_Cubic>]}
        GreenCurve {keys [<at 0.0572 0 RCIM_Cubic>,<at 0.4612 1.2719 RCIM_Cubic>]}
        RedCurve {keys [<at 0.0572 1 RCIM_Cubic>,<at 0.4612 2 RCIM_Cubic>]}
       }}
      CurveIndex=$Particles.Initial.NormalizedChainPosition
     )
     1=&ColorFromCurve(
      ColorCurve=DataInterface() {propertyValues {
        AlphaCurve {keys [<at 0 1 RCIM_Cubic>]}
        BlueCurve {keys [<at 0.0198 0 RCIM_Cubic>,<at 0.2907 0 RCIM_Cubic>]}
        GreenCurve {keys [<at 0.0198 1.2719 RCIM_Cubic>,<at 0.2907 0 RCIM_Cubic>]}
        RedCurve {keys [<at 0.0198 2 RCIM_Cubic>,<at 0.2907 1 RCIM_Cubic>]}
       }}
      CurveIndex=$Particles.Initial.NormalizedChainPosition
     )
     2=?
     3=?
     Index=$Particles.PartitionParticles.Partition
     Num=::NewEnumerator0
    )
    Particles.ColorWispy:LinearColor=&MergeHelper_ColorSwitch(0=<rgba 3 0 0 1>, 1=<rgba 2.199 1.3985 0 1>, 2=?, 3=?, Index=$Particles.PartitionParticles.Partition, Num=::NewEnumerator0)
    Particles.DynamicMaterialParameter8:Vector4f=&MakeVector4(X=1, Y=1, Z=&Lerp_Float(A=3, B=4, Alpha=$System.SmoothDiveAlpha), W=0)
    Particles.RibbonFacing=/Niagara/DynamicInputs/Vectors/CrossProduct(Vector A:Vector3f=$Transient.PhysicsPendulumAxis, Vector B:Vector3f=<0 0 1>, Normalize Incoming Vectors:NiagaraBool=true)
    Particles.RibbonWidth2:NiagaraFloat=&Multiply_Float(A=$Particles.RibbonWidth, B=1)
   ) {isSetParameters=true}
   /SetVariables(
    Particles.Color3:LinearColor=&MergeHelper_ColorSwitch(0=<rgba 1.5 0 0 1>, 1=<rgba 3 1.2863 0 1>, 2=?, 3=?, Index=$Particles.PartitionParticles.Partition, Num=::NewEnumerator0)
    Particles.DynamicMaterialParameter9:Vector4f=&MakeVector4(
     X=&Multiply_Float(
      A=$Engine.System.Age
      B=&Lerp_Float(A=-2, B=2, Alpha=&MakeFloatFromInt(Int=$Particles.Initial.PartitionParticles.Partition))
     )
     Y=&Lerp_Float(A=-3, B=3, Alpha=&MakeFloatFromInt(Int=$Particles.Initial.PartitionParticles.Partition))
     Z=&Multiply_Float(
      A=&FloatFromCurve(FloatCurve=DataInterface() {propertyValues {Curve {keys [<at 0 0.2>,<at 1 0.6908>]}}}, CurveIndex=$Particles.NormalizedChainPosition)
      B=&Lerp_Float(A=0.4, B=1, Alpha=$System.SmoothDiveAlpha)
     )
     W=&Lerp_Float(
      A=&FloatFromCurve(
       FloatCurve=DataInterface() {propertyValues {Curve {keys [{arriveTangent=-2.2554, interpMode=RCIM_Cubic, leaveTangent=-2.2554, tangentMode=RCTM_User, time=0, value=1}, <at 1 0>]}}}
       CurveIndex=$Particles.PartitionParticles.Normalized Index In Partition
      )
      B=&FloatFromCurve(
       FloatCurve=DataInterface() {propertyValues {Curve {keys [
           {arriveTangent=0.0761, interpMode=RCIM_Cubic, leaveTangent=0.0761, tangentMode=RCTM_User, time=0, value=1}
           {arriveTangent=-0.0852, interpMode=RCIM_Cubic, leaveTangent=-0.0852, tangentMode=RCTM_User, time=0.15, value=0.2}
           <at 1 0>
          ]}}}
       CurveIndex=$Particles.PartitionParticles.Normalized Index In Partition
       Scale Curve=10
      )
      Alpha=$System.SmoothDiveAlpha
     )
    )
    Particles.RibbonWidth3:NiagaraFloat=&Multiply_Float(
     A=$Particles.RibbonWidth
     B=&FloatFromCurve(
      FloatCurve=DataInterface() {propertyValues {Curve {keys [
          {arriveTangent=0.357, interpMode=RCIM_Cubic, leaveTangent=0.357, tangentMode=RCTM_User, time=0, value=0.3}
          {arriveTangent=-0.0652, interpMode=RCIM_Cubic, leaveTangent=-0.0652, tangentMode=RCTM_User, time=0.2, value=1}
          {arriveTangent=-0.1821, interpMode=RCIM_Cubic, leaveTangent=-0.1821, tangentMode=RCTM_User, time=0.7, value=0}
         ]}}}
      CurveIndex=$Particles.NormalizedChainPosition
      Scale Curve=&Lerp_Float(A=0.21, B=0.25, Alpha=$System.SmoothDiveAlpha)
     )
    )
   ) {isSetParameters=true}
   /SetVariables(
    Particles.DynamicMaterialParameter14:Vector4f=&MakeVector4(
     X=&FloatFromCurve(
      FloatCurve=DataInterface() {propertyValues {Curve {keys [
          <at 0 0.05 RCIM_Cubic>
          {arriveTangent=-0.1056, interpMode=RCIM_Cubic, leaveTangent=-0.1056, tangentMode=RCTM_User, time=0.0615, value=1}
          {arriveTangent=-2.6928, tangentMode=RCTM_User, time=1, value=0}
         ]}}}
      CurveIndex=$Particles.PartitionParticles.Normalized Index In Partition
      Scale Curve=&Lerp_Float(A=0.7, B=0.83, Alpha=$System.SmoothDiveAlpha)
     )
     Y=0
     Z=0
     W=0
    )
    Particles.RibbonWidth4:NiagaraFloat=&Multiply_Float(
     A=$Particles.RibbonWidth3
     B=&FloatFromCurve(
      FloatCurve=DataInterface() {propertyValues {Curve {keys [
          <at 0 1 RCIM_Cubic>
          {arriveTangent=-0.104, interpMode=RCIM_Cubic, leaveTangent=-0.104, tangentMode=RCTM_User, time=0.05, value=1.7}
          <at 1 1.2 RCIM_Cubic>
         ]}}}
      CurveIndex=$Particles.PartitionParticles.Normalized Index In Partition
     )
    )
   ) {isSetParameters=true}
  ]
  rendererClasses [&NiagaraRibbonRendererProperties]
  renderers [
   {rendererClass=&NiagaraRibbonRendererProperties}
   {rendererClass=&NiagaraRibbonRendererProperties, rendererIndex=1}
  ]
 }
 Splines {
  emitterUpdateScript [
   &EmitterState(
    Inactive Response=::NewEnumerator1
    Life Cycle Mode=::NewEnumerator1
    Loop Behavior=::NewEnumerator1
    Loop Duration Mode=::NewEnumerator1
    UseLoopDelay=false
   ) {module=EmitterLifeCycle}
   &SpawnRate(
    Spawn Probability=0.8
    SpawnRate=&Lerp_Float(A=8, B=12, Alpha=$System.SmoothDiveAlpha)
    Use Spawn Probability=true
   )
  ]
  particleSpawnScript [
   &InitializeParticle(
    Lifetime=&Lerp_Float(A=1.3, B=1, Alpha=$System.SmoothDiveAlpha)
    Sprite Size Mode=::NewEnumerator3
    UsePositionOffset=false
    Write Material Random=true
   )
   &PartitionParticles()
   &InitialMeshOrientation(
    Mesh Orientation Mode=::NewEnumerator2
    Orientation Coordinate Space=::World
    Orientation Vector=<0 0 1>
    Rotation=&RandomRangeVector(Minimum=<0 0 0>, Maximum=<0 0 1>)
    Use Rotation Vector=true
   )
   /SetVariables(Particles.Scale=&Multiply_VectorByFloat(
     Vector=&Lerp_Vector(
      A=<0.7 0.7 2.5>
      B=<0.7 -0.7 2.5>
      Alpha=&MakeFloatFromInt(Int=$Particles.PartitionParticles.Partition)
     )
     Float=&Multiply_Float(
      A=&RandomRangeFloat(Minimum=0.4, Maximum=1)
      B=&Lerp_Float(A=0.6, B=1, Alpha=$System.SmoothDiveAlpha)
     )
    )) {isSetParameters=true}
  ]
  particleUpdateScript [
   &ParticleState()
   &ScaleMeshSize(Scale Factor=&VectorFromCurve(VectorCurve=DataInterface() {propertyValues {
       XCurve {keys [{arriveTangent=3.2161, interpMode=RCIM_Cubic, leaveTangent=3.2161, tangentMode=RCTM_User, time=0, value=0}, <at 1 1>]}
       YCurve {keys [{arriveTangent=3.2161, interpMode=RCIM_Cubic, leaveTangent=3.2161, tangentMode=RCTM_User, time=0, value=0}, <at 1 1>]}
       ZCurve {keys [{arriveTangent=3.2161, interpMode=RCIM_Cubic, leaveTangent=3.2161, tangentMode=RCTM_User, time=0, value=0}, <at 1 1>]}
      }}))
   &SolveForcesAndVelocity(Clamp Velocity=false, Limit Acceleration=false)
   &self:FollowRibbon1Path(
    Index:NiagaraInt32=$Particles.PartitionParticles.Index In Partition
    Offset:NiagaraFloat=&FloatFromCurve(
     FloatCurve=DataInterface() {propertyValues {Curve {keys [
         {arriveTangent=2.4606, interpMode=RCIM_Cubic, leaveTangent=2.4606, tangentMode=RCTM_User, time=0, value=0.1}
         {tangentMode=RCTM_User, time=1, value=1}
        ]}}}
     Scale Curve=0.5
    )
    Particle Attribute Reader=DataInterface() {propertyValues {EmitterBinding {bindingMode=Other, emitterName=NE_LocalSpaceRibbon}}}
    ParticlePerRibbon:NiagaraInt32=$System.ParticlePerRibbon
    Partition:NiagaraInt32=$Particles.PartitionParticles.Partition
   ) {module=FollowRibbon1Path}
   &UpdateMeshOrientation(
    Facing Direction=&AddVectorToPosition(
     Position=Linked() {linkedVariable {name=Particles.RibbonDirectionVector, type=Vector3f}}
     Vector=&Multiply_VectorByFloat(Vector=&RandomVector(Vector Scale=0.1), Float=1)
    )
    Facing Interpolation=false
    Orientation Method=::NewEnumerator1
    Override Quat=false
    TransformFacing=false
    TransformUp=false
    Up Direction:Vector3f=<0 1 0>
   )
   &UpdateMeshOrientation(
    Rotation Rate=&FloatFromCurve(
     FloatCurve=DataInterface() {propertyValues {Curve {keys [
         {arriveTangent=-1.4672, interpMode=RCIM_Cubic, leaveTangent=-1.4672, tangentMode=RCTM_User, time=0, value=1}
         {arriveTangent=-0.5831, tangentMode=RCTM_User, time=1, value=0}
        ]}}}
     Scale Curve=&Multiply_Float(
      A=&Lerp_Float(A=-1, B=1, Alpha=&MakeFloatFromInt(Int=$Particles.Initial.PartitionParticles.Partition))
      B=&Lerp_Float(A=2, B=4, Alpha=$System.SmoothDiveAlpha)
     )
    )
    Rotation Vector=<0 0 1>
   ) {module=UpdateMeshOrientation001}
   /SetVariables(Particles.Color=&MergeHelper_ColorSwitch(0=<rgba 6 0 0 1>, 1=<rgba 5 3.1799 0 1>, 2=?, 3=?, Index=$Particles.PartitionParticles.Partition, Num=::NewEnumerator0)) {isSetParameters=true, module=SetVariables_#7}
   &DynamicMaterialParameters(
    Index 0 Param 1=&RandomRangeFloat(Minimum=5, Maximum=20)
    Index 0 Param 2=1
    Index 0 Param 3=&Multiply_Float(A=$Engine.System.Age, B=1)
    Module.Index0Param0Write=true
    Module.Index0Param1Write=true
    Module.Index0Param2Write=true
    Module.Index0Param3Write=true
   )
  ]
  rendererClasses [&NiagaraMeshRendererProperties]
  renderers [{rendererClass=&NiagaraMeshRendererProperties, rendererIndex=0}]
 }
 TertiaryParticles {
  emitterUpdateScript [
   &EmitterState(
    Inactive Response=::NewEnumerator1
    Life Cycle Mode=::NewEnumerator1
    Loop Behavior=::NewEnumerator1
    Loop Duration Mode=::NewEnumerator1
    UseLoopDelay=false
   ) {module=EmitterLifeCycle}
   &SpawnRate(
    Spawn Probability=0.8
    SpawnRate=&Lerp_Float(A=8, B=30, Alpha=$System.SmoothDiveAlpha)
    Use Spawn Probability=true
   )
   &SpawnRate(
    Spawn Probability=0.8
    SpawnGroup=1
    SpawnRate=&Lerp_Float(A=8, B=30, Alpha=$System.SmoothDiveAlpha)
    Use Spawn Probability=true
   ) {module=SpawnRate001}
   &SpawnRate(SpawnGroup=2, SpawnRate=&Lerp_Float(A=18, B=25, Alpha=$System.SmoothDiveAlpha), Use Spawn Probability=false) {module=SpawnRate002}
  ]
  particleSpawnScript [
   /SetVariables(Particles.VisibilityTag=&SpawnGroup()) {isSetParameters=true}
   &PartitionParticles()
   &InitializeParticle(
    Color Mode=::NewEnumerator0
    Lifetime=&MergeHelper_FloatSwitch(
     0=&Lerp_Float(A=0.5, B=0.5, Alpha=$System.SmoothDiveAlpha)
     1=&Lerp_Float(A=0.5, B=0.5, Alpha=$System.SmoothDiveAlpha)
     2=&Lerp_Float(A=1.5, B=1.3754, Alpha=$System.SmoothDiveAlpha)
     3=0
     4=0
     Index=&SpawnGroup()
     Num=::NewEnumerator3
    )
    Sprite Rotation Angle=&MergeHelper_FloatSwitch(
     0=&RandomRangeFloat(Minimum=0, Maximum=1)
     1=0
     2=?
     3=?
     Index=&SpawnGroup()
     Num=::NewEnumerator0
    )
    Sprite Rotation Mode=::NewEnumerator1
    Sprite Size=&MergeHelper_Vector2Switch(
     0=&Multiply_Vector2D_ByFloat(
      Float=&FloatFromCurve(
       FloatCurve=DataInterface() {propertyValues {Curve {keys [<at 0 2 RCIM_Cubic>, {arriveTangent=6.0011, tangentMode=RCTM_User, time=1, value=10}]}}}
       CurveIndex=&RandomRangeFloat(Minimum=0, Maximum=1)
      )
      Vector2D=<2 2>
     )
     1=&RandomRangeVector2D(Minimum=<5 15>, Maximum=<10 25>)
     2=?
     3=?
     4=?
     Index=&SpawnGroup()
     Num=::NewEnumerator0
    )
    Sprite Size Mode=::NewEnumerator4
    UsePositionOffset=false
    Write Material Random=true
   )
   &InitialMeshOrientation(
    Mesh Orientation Mode=::NewEnumerator2
    Orientation Coordinate Space=::World
    Orientation Vector=<0 0 1>
    Rotation=&RandomRangeVector(Minimum=<0 0 0>, Maximum=<0 0 1>)
    Use Rotation Vector=true
   )
  ]
  particleUpdateScript [
   &ParticleState()
   &SpriteRotationRate(Rotation Rate=&MergeHelper_FloatSwitch(
     0=&SelectFloatFromArray(
      Array Sampling Mode=::NewEnumerator0
      Direct Array Index=?
      Float Selection Array=DataInterface() {propertyValues {FloatData [350,-350]}}
      Random Seed=0
     )
     1=0
     2=&Multiply_Float(
      A=&FloatFromCurve(FloatCurve=DataInterface() {propertyValues {Curve {keys [<at 0 1 RCIM_Cubic>,<at 1 0>]}}})
      B=&Lerp_Float(A=-180, B=180, Alpha=&MakeFloatFromInt(Int=$Particles.PartitionParticles.Partition))
     )
     3=?
     Index=&SpawnGroup()
     Num=::NewEnumerator1
    ))
   &ScaleSpriteSize(Uniform Curve Sprite Scale=DataInterface() {propertyValues {Curve {keys [
        {arriveTangent=6.8926, interpMode=RCIM_Cubic, leaveTangent=6.8926, tangentMode=RCTM_User, time=0, value=0}
        {interpMode=RCIM_Cubic, tangentMode=RCTM_User, time=0.2, value=1}
        {arriveTangent=-3.4257, tangentMode=RCTM_User, time=1, value=0}
       ]}}})
   &SolveForcesAndVelocity(Clamp Velocity=false, Limit Acceleration=false)
   &self:FollowRibbon1Path(
    Index:NiagaraInt32=$Particles.PartitionParticles.Index In Partition
    Offset:NiagaraFloat=&MergeHelper_FloatSwitch(
     0=&FloatFromCurve(
      FloatCurve=DataInterface() {propertyValues {Curve {keys [
          {arriveTangent=1.739, interpMode=RCIM_Cubic, leaveTangent=1.739, tangentMode=RCTM_User, time=0, value=0.3}
          {tangentMode=RCTM_User, time=1, value=1}
         ]}}}
      Scale Curve=&RandomRangeFloat(Minimum=0.1, Maximum=0.7)
     )
     1=&FloatFromCurve(
      FloatCurve=DataInterface() {propertyValues {Curve {keys [
          {arriveTangent=1.739, interpMode=RCIM_Cubic, leaveTangent=1.739, tangentMode=RCTM_User, time=0, value=0.3}
          {tangentMode=RCTM_User, time=1, value=1}
         ]}}}
      Scale Curve=&RandomRangeFloat(Minimum=0.1, Maximum=0.7)
     )
     2=&FloatFromCurve(
      FloatCurve=DataInterface() {propertyValues {Curve {keys [{arriveTangent=1.9463, interpMode=RCIM_Cubic, time=0, value=0}, {tangentMode=RCTM_User, time=1, value=1}]}}}
      Scale Curve=0.6
     )
     3=?
     Index=&SpawnGroup()
     Num=::NewEnumerator1
    )
    Particle Attribute Reader=DataInterface() {propertyValues {EmitterBinding {bindingMode=Other, emitterName=NE_LocalSpaceRibbon}}}
    ParticlePerRibbon:NiagaraInt32=$System.ParticlePerRibbon
    Partition:NiagaraInt32=$Particles.PartitionParticles.Partition
   ) {module=FollowRibbon1Path}
   &self:NoiseLocation(
    Amplitude:Vector3f=&Multiply_VectorByFloat(
     Vector=<1 1 1>
     Float=&FloatFromCurve(
      FloatCurve=DataInterface() {propertyValues {Curve {keys [
          {interpMode=RCIM_Cubic, tangentMode=RCTM_User, time=0, value=0}
          {interpMode=RCIM_Cubic, tangentMode=RCTM_User, time=0.2, value=1}
          {tangentMode=RCTM_User, time=1, value=0}
         ]}}}
      Scale Curve=&RandomRangeFloat(Minimum=1, Maximum=4)
     )
    )
    Offset:Vector3f=&VectorFromFloat(Value=&RandomRangeFloat(Minimum=0, Maximum=200))
    Scale:Vector3f=&Multiply_VectorByFloat(
     Vector=<1 1 1>
     Float=&MergeHelper_FloatSwitch(
      0=&RandomRangeFloat(Minimum=400, Maximum=400)
      1=&RandomRangeFloat(Minimum=400, Maximum=400)
      2=0
      3=?
      Index=&SpawnGroup()
      Num=::NewEnumerator1
     )
    )
   ) {module=NoiseLocation}
   &UpdateMeshOrientation(
    Facing Direction=&AddVectorToPosition(
     Position=Linked() {linkedVariable {name=Particles.RibbonDirectionVector, type=Vector3f}}
     Vector=&Multiply_VectorByFloat(Vector=&RandomVector(Vector Scale=0.1), Float=1)
    )
    Facing Interpolation=false
    Orientation Method=::NewEnumerator1
    Override Quat=false
    TransformFacing=false
    TransformUp=false
    Up Direction:Vector3f=<0 1 0>
   )
   &UpdateMeshOrientation(
    Rotation Rate=&FloatFromCurve(
     FloatCurve=DataInterface() {propertyValues {Curve {keys [
         {arriveTangent=-1.4672, interpMode=RCIM_Cubic, leaveTangent=-1.4672, tangentMode=RCTM_User, time=0, value=1}
         {arriveTangent=-0.5831, tangentMode=RCTM_User, time=1, value=0}
        ]}}}
     Scale Curve=&Multiply_Float(
      A=&Lerp_Float(A=-1, B=1, Alpha=&MakeFloatFromInt(Int=$Particles.Initial.PartitionParticles.Partition))
      B=&Lerp_Float(A=2, B=4, Alpha=$System.SmoothDiveAlpha)
     )
    )
    Rotation Vector=<0 0 1>
   ) {module=UpdateMeshOrientation001}
   /SetVariables(
    Particles.Color=&MultiplyLinearColor(
     Linear Color=$Particles.RibbonColor
     Scale Factor=&MakeVector4_FromLinearColor(LinearColor=&ColorFromCurve(ColorCurve=DataInterface() {propertyValues {
         AlphaCurve {keys [<at 0.0069 0>,<at 0.1431 1>,<at 1 0 RCIM_Cubic>]}
         BlueCurve {keys [<at 0.0035 10>,<at 1 0 RCIM_Cubic>]}
         GreenCurve {keys [<at 0.0035 10>,<at 1 0 RCIM_Cubic>]}
         RedCurve {keys [<at 0.0035 10>,<at 1 0 RCIM_Cubic>]}
        }}))
    )
    Particles.DotColors:LinearColor=&SelectLinearColorFromArray(
     Array Sampling Mode=::NewEnumerator0
     Direct Array Index=?
     Color Selection Array=DataInterface() {propertyValues {ColorData [<rgba 3 0.0058 0 0.2>, <rgba 2 0.3793 0 0.5217>]}}
     Random Seed=0
    )
    Particles.SpriteAlign:Vector3f=&Subtract_Vector(
     A=Linked() {linkedVariable {name=Particles.Position, type=NiagaraPosition}}
     B=Linked() {linkedVariable {name=Particles.Previous.Position, type=NiagaraPosition}}
    )
   ) {isSetParameters=true, module=SetVariables_#7}
   &DynamicMaterialParameters(
    Index 0 Param 1=&RandomRangeFloat(Minimum=5, Maximum=20)
    Index 0 Param 2=1
    Index 0 Param 3=&Multiply_Float(A=$Engine.System.Age, B=1)
    Module.Index0Param0Write=true
    Module.Index0Param1Write=true
    Module.Index0Param2Write=true
    Module.Index0Param3Write=true
   )
   &DynamicMaterialParameters(
    Index 0 Param 1=0.3753
    Index 0 Param 2=-0.3612
    Module.Index0Param0Write=true
    Module.Index0Param1Write=true
    Module.Index0Param2Write=true
    Module.Index0Param3Write=true
   ) {module=DynamicMaterialParameters001}
   &CameraOffset(Camera Offset Amount=&MergeHelper_FloatSwitch(0=20, 1=20, 2=-5, 3=?, Index=&SpawnGroup(), Num=::NewEnumerator1))
   &SubUVAnimation(
    Output Alternate Renderer Binding=true
    "Sprite Renderer "=DataInterface() {propertyValues {SpriteRenderer=&self:TertiaryParticles_0.NiagaraSpriteRendererProperties_3}}
    UseEndFrame=false
    UseStartFrame=false
   )
   /SetVariables(
    Particles.SmokeColor:LinearColor=&Multiply_LinearColorByFloat(
     Linear Color=&MergeHelper_ColorSwitch(0=<rgba 10 0 0 1>, 1=<rgba 10 4.0987 0 1>, 2=?, 3=?, Index=$Particles.PartitionParticles.Partition, Num=::NewEnumerator0)
     Float=&FloatFromCurve(
      FloatCurve=DataInterface() {propertyValues {Curve {keys [{interpMode=RCIM_Cubic, tangentMode=RCTM_User, time=0, value=0}, <at 0.5 0.5121 RCIM_Cubic>, {tangentMode=RCTM_User, time=1, value=0}]}}}
      Scale Curve=&Lerp_Float(A=0.3, B=0.35, Alpha=$System.SmoothDiveAlpha)
     )
    )
    Particles.SmokeSize:Vector2f=&Multiply_Vector2D_ByFloat(
     Float=&Multiply_Float(
      A=&RandomRangeFloat(Minimum=83.2516, Maximum=61.154)
      B=&Lerp_Float(A=1, B=1.9409, Alpha=$System.SmoothDiveAlpha)
     )
     Vector2D=&Vector2DFromFloat(Value=&FloatFromCurve(FloatCurve=DataInterface() {propertyValues {Curve {keys [
           {arriveTangent=1.4971, interpMode=RCIM_Cubic, leaveTangent=1.4971, tangentMode=RCTM_User, time=0, value=0}
           {tangentMode=RCTM_User, time=1, value=1}
          ]}}}))
    )
    Particles.SmokeUVScale:Vector2f=&Lerp_Vector2(A=<-1 1>, B=<1 1>, Alpha=&MakeFloatFromInt(Int=$Particles.PartitionParticles.Partition))
   ) {isSetParameters=true}
  ]
  rendererClasses [&NiagaraSpriteRendererProperties]
  renderers [
   {rendererClass=&NiagaraSpriteRendererProperties}
   {rendererClass=&NiagaraSpriteRendererProperties, rendererIndex=1}
   {rendererClass=&NiagaraSpriteRendererProperties, rendererIndex=2}
  ]
 }
}
summary {
 systemName=NS_Contrail_Season40
 emitters [
  {emitterName=NE_LocalSpaceRibbon, rendererClasses [&NiagaraRibbonRendererProperties]}
  {emitterName=ExtraSplines, rendererClasses [&NiagaraMeshRendererProperties]}
  {emitterName=Splines, rendererClasses [&NiagaraMeshRendererProperties]}
  {emitterName=FeetSpline, rendererClasses [&NiagaraMeshRendererProperties]}
  {
   emitterName=Hand_Feet_Flares
   rendererClasses [
    &NiagaraMeshRendererProperties
    &NiagaraLightRendererProperties
    &NiagaraSpriteRendererProperties
   ]
  }
  {emitterName=TertiaryParticles, rendererClasses [&NiagaraSpriteRendererProperties]}
  {emitterName=DigitalSquares001, rendererClasses [&NiagaraSpriteRendererProperties]}
 ]
 userVariables [
  User.bIsFrontEnd {type=NiagaraBool}
  User.bIsFrontEndPreview {type=NiagaraBool, defaultValue=true}
  User.Debug {type=NiagaraBool}
  User.FXColorA {type=LinearColor, description="", defaultValue {struct=LinearColor, value=<rgba 1 0.012 0 1>}}
  User.FXColorB {type=LinearColor, description="", defaultValue {struct=LinearColor, value=<rgba 0.2698 0.001 0.005 1>}}
  User.ParaGlideLeanAlpha {type=NiagaraFloat, defaultValue=1, description=""}
 ]
}
